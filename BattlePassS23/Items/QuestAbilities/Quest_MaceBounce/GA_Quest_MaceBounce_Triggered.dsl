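; GA_Quest_MaceBounce_Triggered is a quest gameplay ability that reports progress for the Mace Bounce quest.
; On activation it finds the avatar's player controller and gets its Athena quest manager.
; It sends one complex custom stat event carrying the Additional Source Tags variable and empty target tags, then ends immediately.
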(event Ability|EventActivateAbility
  (Quest|SendComplexCustomStatEvent (Quests|GetQuestManager (Fortnite|GetFortPlayerControllerfromActor (Ability|GetAvatarActorfromActorInfo)) "Athena") 0 (Variables|Default|GetAdditional%20Source%20Tags) "(GameplayTags=)" 1)
  (Ability|EndAbility))
